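; Weapon Blueprint for the PocketScrunchie pickaxe that feeds the current swing into its animation trail VFX.
; On weapon fire FX it calls the parent, then looks up the instigator's swing montage section index through a utility function.
; It writes that index to the trail particle component's SwingSectionIndex parameter, or -1 when the cast fails.
; BeginPlay and Tick are empty, and the construction script only calls the parent.
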(event EventBeginPlay)

(event EventTick
  (params (DeltaSeconds :type Float%20%28single-precision%29)))

(event EventOnPlayWeaponFireFX
  (params (bPersistentFire :type Boolean) (bSecondaryFire :type Boolean))
  (bind _returnvalue (CallFunction|GetAnimTrailVFXComponent))
  (|CallParentFunction false false)
  (bind _section_index (Utilities|SetupSwingMontageSectionIndex (Variables|Actor|GetInstigator))
    (:Out
      (Effects|Components|ParticleSystem|SetIntParameter _returnvalue "SwingSectionIndex" _section_index))
    (:Cast%20Failed%20Out
      (Effects|Components|ParticleSystem|SetIntParameter _returnvalue "SwingSectionIndex" -1))))

(fn ConstructionScript
  (|CallParentFunction))
