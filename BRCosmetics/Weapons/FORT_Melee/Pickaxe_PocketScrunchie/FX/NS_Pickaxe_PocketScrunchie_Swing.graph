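# Niagara swing effect for the PocketScrunchie pickaxe, pairing a short-lived ribbon trail with small falling stickers.
# FN_AnimTrail follows the system skeletal mesh through two ribbon renderers, shaded by color curve, time of day and TrailFade.
# Stickers spawn per distance moved, placed between the simulation position and a sampled mesh point, then drift with curl noise, gravity and drag.
# One modulo-3 index selects each sticker's color and SubUV frame and derives its visibility tag.
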NiagaraSystem /BRCosmetics/Weapons/FORT_Melee/Pickaxe_PocketScrunchie/FX/NS_Pickaxe_PocketScrunchie_Swing
dependencies {
 usedDataInterfaces [
  &NiagaraDataInterfaceCurve
  &NiagaraDataInterfaceColorCurve
  &NiagaraDataInterfaceArrayColor
  &NiagaraDataInterfaceSpriteRendererInfo
 ]
 usedDynamicInputs [&self:SectionIndexFloatSelector]
 usedModules [
  &self:SystemSpawnScript.SystemScriptSource.SystemScriptGraph.NiagaraNodeAssignment_0.SetVariables_#1
  &SystemState
  &self:SystemSpawnScript.SystemScriptSource.SystemScriptGraph.NiagaraNodeAssignment_1.SetVariables_#2
  &self:SystemSpawnScript.SystemScriptSource.SystemScriptGraph.NiagaraNodeAssignment_3.SetVariables_#3
 ]
}
emitters {
 FN_AnimTrail {
  emitterSpawnScript [&FN_RibbonIDOffsetUtility()]
  emitterUpdateScript [
   &EmitterState(Life Cycle Mode=::NewEnumerator1, Loop Duration=1, UseLoopDelay=false)
   &FN_SpawnAnimTrail(Spawn Interval=0)
  ]
  particleSpawnScript [
   &FN_InitializeAnimTrailBoneIndecies()
   &FN_InitializeAnimTrail(Color=<rgba 1 1 1 0.6>, Lifetime=0.12, SkeletalMesh=$System.Skeletal Mesh)
   &AddVelocity(Velocity=<0 400 0>) {enabled=false}
  ]
  particleUpdateScript [
   &SolveForcesAndVelocity(Clamp Velocity=false, Limit Acceleration=false)
   &ScaleRibbonWidth(Ribbon Width Scale=&FloatFromCurve(FloatCurve=DataInterface() {propertyValues {Curve {keys [
         {arriveTangent=-1.2579, interpMode=RCIM_Cubic, time=0, value=1}
         {arriveTangent=-3.0042, interpMode=RCIM_Cubic, leaveTangent=-3.0042, tangentMode=RCTM_User, time=1, value=0}
        ]}}}))
   &FN_UpdateAnimTrail(Locked Particle RibbonWidth=1, SkeletalMesh=$System.Skeletal Mesh) {module=FN_UpdateAnimTrail001}
   &ScaleColor(
    Scale Alpha=&Multiply_Float(
     A=$System.TrailFade
     B=&FN_Select_Float_from_Time_Of_Day(Morning=0.9, Day=1, Evening=1, Night=0.8)
    )
    Scale RGB=&Multiply_Vector(
     A=&MakeVectorFromLinearColorRGB(Color=&ColorFromCurve(ColorCurve=DataInterface() {propertyValues {
         AlphaCurve {keys [<at 0 1 RCIM_Cubic>]}
         BlueCurve {keys [
           {arriveTangent=0.96, interpMode=RCIM_Cubic, time=0.1316, value=0.1805}
           <at 0.462 0.2505>
           {interpMode=RCIM_Cubic, leaveTangent=-0.1644, time=1, value=0.7188}
          ]}
         GreenCurve {keys [
           {arriveTangent=0.0419, interpMode=RCIM_Cubic, time=0.1316, value=0.1163}
           <at 0.462 0.1337>
           {interpMode=RCIM_Cubic, leaveTangent=0.2347, time=1, value=0.2547}
          ]}
         RedCurve {keys [{arriveTangent=-0.4464, interpMode=RCIM_Cubic, time=0.1316, value=0.375}, <at 0.462 0.375>, <at 1 0.1203 RCIM_Cubic>]}
        }}))
     B=&MakeVectorFromLinearColorRGB(Color=&FN_Select_Color_from_Time_Of_Day(
       Use Manual Time of Day Thresholds=true
       Morning=<rgba 0.493 0.4792 0.7812 1>
       Day=<rgba 1 1 1 1>
       Evening=<rgba 0.5469 0.5284 0.7865 1>
       Night=<rgba 0.3581 0.3594 0.7812 1>
       Morning Threshold=9
       Day Threshold=17
       Evening Threshold=21
      ))
    )
    ScaleA=true
    ScaleRGB=true
   )
  ]
  rendererClasses [&NiagaraRibbonRendererProperties]
  renderers [
   {rendererClass=&NiagaraRibbonRendererProperties}
   {rendererClass=&NiagaraRibbonRendererProperties, rendererIndex=1}
  ]
 }
 Stickers {
  emitterUpdateScript [
   &EmitterState(Life Cycle Mode=::NewEnumerator1, Loop Duration=1, UseLoopDelay=false)
   &SpawnPerUnit(Module.Use Max Movement Threshold=true, Module.Use Movement Tolerance=false, Spawn Spacing=80, Use Spawn Probability=false)
  ]
  particleSpawnScript [
   /SetVariables(Particles.CustomIndex:NiagaraInt32=&ModuloInt(
     A=&Add_Integer(
      A=&RandomRangeInt(Minimum=0, Maximum=1)
      B=&ReturnParticleIDAttributes()
     )
     B=3
    )) {isSetParameters=true}
   /SetVariables(Particles.VisibilityTag=&MakeCustomIntFromBool(
     Bool=&SetBoolByIntComparison(A=$Particles.CustomIndex, B=1, Comparison Type=::NewEnumerator0)
     True INT=1
     False INT=0
    )) {isSetParameters=true}
   &InitializeParticle(
    Color=&SelectLinearColorFromArray(
     Array Sampling Mode=::NewEnumerator1
     Direct Array Index=$Particles.CustomIndex
     Color Selection Array=DataInterface() {propertyValues {ColorData [<rgba 0.6719 0.2694 0.0787 1>, <rgba 1 0.1562 0.1661 1>, <rgba 0 0.5081 1 1>]}}
     Random Seed=?
    )
    Lifetime Max=0.6
    Lifetime Min=0.3
    Lifetime Mode=::NewEnumerator1
    Sprite Rotation Angle=&RandomRangeFloat(Minimum=-30, Maximum=30)
    Sprite Rotation Mode=::NewEnumerator0
    Sprite Size Mode=::NewEnumerator1
    Uniform Sprite Size Min=3
    UsePositionOffset=false
    Write Material Random=true
   )
   &SubUVAnimation(
    Direct Frame Index:NiagaraFloat=&MakeFloatFromInt(Int=$Particles.CustomIndex)
    "Sprite Renderer "=DataInterface() {propertyValues {SpriteRenderer=&self:Stickers_4.NiagaraSpriteRendererProperties_0}}
    SubUV Animation Mode:ENiagara_SubUVLookupModeV2=::NewEnumerator5
   )
   &SkeletalMeshLocation(
    Bone / Socket Index=1
    MaskSpawnGroup=false
    Mesh Sampling Type=::NewEnumerator5
    Position Sampling=::NewEnumerator2
    Skeletal Mesh=$System.Skeletal Mesh
    Socket Sampling Mode=::NewEnumerator5
   )
   &AddVelocity(
    Rotation Coordinate Space=::World
    Velocity=&RandomRangeVector(Minimum=<-50 -50 50>, Maximum=<50 50 100>)
   )
   /SetVariables(Particles.Position=&LerpPosition(
     A=/Niagara/DynamicInputs/Helpers/SimulationPosition()
     B=$Particles.SkeletalMeshLocation.SampledPosition
     Alpha=&RandomRangeFloat(Minimum=0, Maximum=1)
    )) {isSetParameters=true}
  ]
  particleUpdateScript [
   &ParticleState()
   &ScaleSpriteSize(Uniform Curve Sprite Scale=DataInterface() {propertyValues {Curve {keys [{arriveTangent=1.4286, time=0.1, value=1}, <at 0.5 1>, <at 1 0.1>]}}})
   &CurlNoiseForce(Noise Frequency=30, Noise Strength=&RandomRangeFloat(Minimum=20, Maximum=50), Pan Noise=false)
   &GravityForce(Gravity=&RandomRangeVector(Minimum=<0 0 -200>, Maximum=<1 1 -300>))
   &Color(Color=&MultiplyLinearColor(
     Linear Color=$Particles.Initial.Color
     Scale Factor=&MakeVector4_FromLinearColor(LinearColor=&FN_Select_Color_from_Time_Of_Day(Morning=<rgba 1 1 1 1>, Day=<rgba 1 1 1 1>, Evening=<rgba 1 0.7874 0.7067 1>, Night=<rgba 1 1 1 1>))
    ))
   &Drag(
    Drag=&RandomRangeFloat(Minimum=1, Maximum=3)
    Rotational Drag=0.2
    Use Linear Drag=true
    Use Rotational Drag=true
   )
   &SolveForcesAndVelocity(Clamp Velocity=false, Limit Acceleration=false)
  ]
  rendererClasses [&NiagaraSpriteRendererProperties]
  renderers [
   {rendererClass=&NiagaraSpriteRendererProperties}
   {rendererClass=&NiagaraSpriteRendererProperties, rendererIndex=1}
  ]
 }
}
summary {
 systemName=NS_Pickaxe_PocketScrunchie_Swing
 emitters [
  {emitterName=FN_AnimTrail, rendererClasses [&NiagaraRibbonRendererProperties]}
  {emitterName=Stickers, rendererClasses [&NiagaraSpriteRendererProperties]}
 ]
 userVariables [
  User.bIsFrontEnd {type=NiagaraBool}
  User.bIsFrontEndPreview {type=NiagaraBool}
  User.RibbonIDOffset {type=NiagaraInt32}
  User.SkeletalMesh {
   type=NiagaraDataInterfaceSkeletalMesh
   defaultValue {
    struct=NiagaraExt_VariableValue_DataInterface
    value {dataInterface=&self:NiagaraDataInterfaceSkeletalMesh_0}
   }
  }
  User.SwingSectionIndex {type=NiagaraInt32}
 ]
}
