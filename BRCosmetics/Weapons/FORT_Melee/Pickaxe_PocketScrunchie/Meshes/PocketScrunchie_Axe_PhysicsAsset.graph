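# Physics asset for the PocketScrunchie pickaxe, a two-handed melee weapon cosmetic.
# It defines a single capsule body on the Melee_TwoHanded_Handle bone, slightly offset and rotated to lie along that part of the mesh.
# It has no constraints, so it provides one simple collision shape with no articulated or jointed parts.
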PhysicsAsset /BRCosmetics/Weapons/FORT_Melee/Pickaxe_PocketScrunchie/Meshes/PocketScrunchie_Axe_PhysicsAsset
bodies {Melee_TwoHanded_Handle [{shapeType=Capsule, length=16.9755, radius=3.6568, center=<-0.0347 -0.5703 7.1575>, rotation=<rot -0.1057 -90 -0.0333>}]}
constraints []
